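; Burst gameplay cue for the first ground attack of the grapple knife.
; On burst it clears the IsSecondAttack flag on the first Niagara component and shows the target pawn's controller a local sound indicator tied to the knife swing cue.
; If the pawn is driving a vehicle, it re-orients that effect from the FX_Spine socket's up vector with a half-turn yaw, then shifts it down along its local Z.
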(fn OnBurst(Niagara)
  (params (MyTarget :type Actor%20Object%20Reference) (Parameters :type Gameplay%20Cue%20Parameters%20Structure%20%28by%20ref%29) (ParticleComponents :type Array%20of%20Niagara%20Particle%20System%20Component%20Object%20References) (AudioComponents :type Array%20of%20Audio%20Component%20Object%20References) (BurstCameraShakeInstance :type Camera%20Shake%20Base%20Object%20Reference) (BurstDecalInstance :type Decal%20Actor%20Object%20Reference))
  (bind _output (Utilities|Array|Get(acopy) ParticleComponents))
  (Niagara|SetNiagaraVariable(Bool) _output "IsSecondAttack" false)
  (bind _asfort_player_pawn (Utilities|Casting|CastToFortPlayerPawn MyTarget))
  (Fortnite|ShowSoundIndicatorLocal 0 0 (Transformation|GetActorLocation _asfort_player_pawn) 7e+03 "Gunshot" (Utilities|Array|MakeArray (Fortnite|GetFortPlayerControllerfromActor _asfort_player_pawn)) 0 "(R=0.000000,G=0.000000,B=0.000000,A=0.000000)" "" "/GrappleWeapon/Assets/Audio/GrappleKnife_Swing_Cue.GrappleKnife_Swing_Cue")
  (if (Vehicle|IsDrivingVehicle _asfort_player_pawn)
    (Transformation|SetWorldRotation _output (Math|Rotator|CombineRotators (Math|Rotator|MakeRotfromZ (Math|Vector|GetUpVector (Transformation|GetSocketRotation (Variables|/Script/Engine.Character|GetMesh _asfort_player_pawn) "FX_Spine"))) "0.000000,-180.000000,0.000000") false false)
    (Transformation|AddLocalOffset _output "0.000000,0.000000,-45.000000" false false)
    (else
      _))
  (return))
